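# DarkMetal is a Default Lit surface material in the temp folder of the Asteria RealityFortress environment set, named as a dark metal finish.
# It is flagged for Nanite use and holds only a dark blue-grey vector parameter and three scalar constants, with no textures.
# The graph shows no connections, so which material inputs these values drive is not visible.
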Material /Game/Environments/Asteria/Sets/RealityFortress/Materials/temp/DarkMetal
graph {
 properties {bUsedWithNanite=true, pixelDepthOffsetMode=PDOM_Legacy, usedShadingModels=Default Lit}
 nodes [
  Constant() {r=0.504}
  Constant() {r=0.317}
  Constant() {r=0.4}
  VectorParameter() {defaultValue=<rgba 0.1254 0.1308 0.151 0>}
 ]
}
