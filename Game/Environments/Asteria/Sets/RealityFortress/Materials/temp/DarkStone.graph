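# DarkStone is a simple Default Lit material in a temp folder of the RealityFortress environment set, flagged for use with Nanite.
# It holds a dark, slightly blue-grey color vector parameter and two scalar constants, with no textures or math.
# The graph shows no links from these nodes to material outputs, so which surface properties they drive is not stated.
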Material /Game/Environments/Asteria/Sets/RealityFortress/Materials/temp/DarkStone
graph {
 properties {bUsedWithNanite=true, pixelDepthOffsetMode=PDOM_Legacy, usedShadingModels=Default Lit}
 nodes [
  Constant() {r=0.65}
  Constant() {r=0.35}
  VectorParameter() {defaultValue=<rgba 0.1192 0.1336 0.151 0>}
 ]
}
